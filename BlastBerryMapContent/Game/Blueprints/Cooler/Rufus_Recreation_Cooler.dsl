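; Cooler Blueprint in the Rufus Recreation set, a lootable container that extends a parent searchable container class.
; Construction runs the parent logic, then adds the tag "Lackluster.Material.TieredAmmoAthena" to the actor's tags without duplicates.
; The loot event only forwards to the parent, so all looting behavior is inherited.
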(event EventOnLoot
  (|CallParentFunction))

(fn ConstructionScript
  (|CallParentFunction)
  (Utilities|Array|AddUnique (Variables|Actor|GetTags) "Lackluster.Material.TieredAmmoAthena"))
